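# Niagara vehicle cosmetic system for the Lantern boost, trailing two sprite emitters from the vehicle's skeletal mesh sockets.
# Sprite_01 makes large, warm-tinted, SubUV-animated sprites that grow, fade and drift upward; Sprite_02 makes small, fast streaks that pulse in brightness and fall.
# Both partition particles across ports, inherit decaying source movement, and switch tuning on bIsFrontendPreview.
# VC_Mask modules add per-game variants; painted-variant overrides are disabled.
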NiagaraSystem /VehicleCosmetics/Boosts/Boost_Lantern/FX/NS_Boost_Lantern
dependencies {
 usedDataInterfaces [
  &NiagaraDataInterfaceCurve
  &NiagaraDataInterfaceVectorCurve
  &NiagaraDataInterfaceSpriteRendererInfo
  &NiagaraDataInterfaceSkeletalMesh
  &NiagaraDataInterfaceArrayFloat3
 ]
 usedDynamicInputs [&SelectVectorFromArray]
 usedModules [
  &self:SystemSpawnScript.SystemScriptSource.SystemScriptGraph.NiagaraNodeAssignment_0.SetVariables_#1
  &self:SystemSpawnScript.SystemScriptSource.SystemScriptGraph.NiagaraNodeAssignment_1.SetVariables_#2
  &self:SystemSpawnScript.SystemScriptSource.SystemScriptGraph.NiagaraNodeAssignment_2.SetVariables_#3
  &SystemState
 ]
}
emitters {
 NE_Lantern_Sprite_01 {
  emitterUpdateScript [
   &EmitterState()
   &VC_FilteredSocketCount()
   &SpawnRate(
    SpawnRate=&Multiply_Float(
     A=&MakeCustomFloatFromBool(Bool=$User.bIsFrontendPreview, True Float=30, False Float=50)
     B=/VehicleCosmetics/FX/Niagara/Modules/VC_Mask_Floats(Coro Value:NiagaraFloat=1, FrontEnd Value:NiagaraFloat=1, FN Value:NiagaraFloat=0.5)
    )
    Use Spawn Probability=false
   )
  ]
  particleSpawnScript [
   &VC_PaintedVariantsOverride() {enabled=false}
   &InitializeParticle(
    Color=&MakeCustomLinearColorFromBool(Bool=$User.bIsFrontendPreview, True Color=<rgba 0.8125 0 0.0948 1>, False Color=<rgba 1 0.3 0.15 1>)
    Lifetime=&MakeCustomFloatFromBool(
     Bool=$User.bIsFrontendPreview
     True Float=&RandomRangeFloat(Minimum=0.8, Maximum=1.5)
     False Float=&RandomRangeFloat(Minimum=1, Maximum=1.5)
    )
    Sprite Size Mode=::NewEnumerator3
    Uniform Sprite Size=&MakeCustomFloatFromBool(
     Bool=$User.bIsFrontendPreview
     True Float=&RandomRangeFloat(Minimum=50, Maximum=70)
     False Float=&RandomRangeFloat(Minimum=80, Maximum=120)
    )
    UsePositionOffset=false
    Write Material Random=true
   )
   &PartitionParticles(Number Of Partitions=$Emitter.PortCount)
   &SkeletalMeshLocation(
    Bone / Socket Index=$Particles.PartitionParticles.Partition
    Bone / Socket Sampling Mode=::NewEnumerator5
    MaskSpawnGroup=false
    Mesh Sampling Type=::NewEnumerator6
    Skeletal Mesh=$System.SkeletalMesh
   )
   &AddVelocity(Velocity=&VC_Mask_Vectors(
     Coro Value:Vector3f=<-400 0 0>
     FrontEnd Value:Vector3f=&MakeCustomVectorFromBool(
      Bool=$User.bIsFrontendPreview
      True Vector=&RandomRangeVector(Minimum=<-300 30 20>, Maximum=<-500 -30 -20>)
      False Vector=&RandomRangeVector(Minimum=<-200 30 50>, Maximum=<-300 -30 -50>)
     )
     FN Value:Vector3f=<-1200 0 0>
    ))
   &ShapeLocation(
    MaskSpawnGroup=false
    Override Seed=false
    Sphere Radius=&MakeCustomFloatFromBool(Bool=$User.bIsFrontendPreview, True Float=20, False Float=50)
    UseOwnerScale=true
   )
   /VehicleCosmetics/FX/VC_FXParents/Modules/DLR_ProductPreviews(GameModeID:NiagaraInt32=$System.GameModeID, bPreviewVehicleSpeed:NiagaraBool=$System.bPreviewVehicleSpeed)
  ]
  particleUpdateScript [
   &ParticleState()
   &ScaleSpriteSize(
    Scale Sprite Size Mode=::NewEnumerator0
    Uniform Scale Factor=&MakeCustomFloatFromBool(
     Bool=$User.bIsFrontendPreview
     True Float=&FloatFromCurve(FloatCurve=DataInterface() {propertyValues {Curve {keys [<at 0 0.4 RCIM_Cubic>, {arriveTangent=0.6, interpMode=RCIM_Cubic, leaveTangent=0.6, time=0.2, value=0.8}, <at 1 1 RCIM_Cubic>]}}})
     False Float=&FloatFromCurve(FloatCurve=DataInterface() {propertyValues {Curve {keys [
          {arriveTangent=2.9579, interpMode=RCIM_Cubic, leaveTangent=2.9579, tangentMode=RCTM_User, time=0, value=0}
          {arriveTangent=1, interpMode=RCIM_Cubic, leaveTangent=1, time=0.3, value=0.75}
          <at 1 1 RCIM_Cubic>
         ]}}})
    )
   )
   &ScaleColor(
    Scale Alpha=&FloatFromCurve(FloatCurve=DataInterface() {propertyValues {Curve {keys [
         <at 0 1>
         {arriveTangent=-0.1877, interpMode=RCIM_Cubic, leaveTangent=-0.1877, tangentMode=RCTM_User, time=0.5, value=1}
         <at 1 0 RCIM_Cubic>
        ]}}})
    Scale RGB=&MakeCustomVectorFromBool(
     Bool=$User.bIsFrontendPreview
     True Vector=&VectorFromCurve(
      VectorCurve=DataInterface() {propertyValues {XCurve {keys [<at 0 1 RCIM_Cubic>,<at 1 1.5 RCIM_Cubic>]}, YCurve {keys [<at 0 1>,<at 1 1>]}, ZCurve {keys [<at 0 1>,<at 1 1>]}}}
      Scale Curve=<2.5 1 1>
     )
     False Vector=&VectorFromCurve(VectorCurve=DataInterface() {propertyValues {XCurve {keys [<at 0 1 RCIM_Cubic>,<at 0.5 1.5>,<at 1 2 RCIM_Cubic>]}, YCurve {keys [<at 0 1>,<at 1 1>]}, ZCurve {keys [<at 0 1>,<at 1 1>]}}})
    )
    ScaleA=true
    ScaleRGB=true
   )
   &AccelerationForce(
    Acceleration=&MakeCustomVectorFromBool(
     Bool=$User.bIsFrontendPreview
     True Vector=&RandomRangeVector(Minimum=<0 0 50>, Maximum=<0 0 200>, Evaluation Type=::NewEnumerator1)
     False Vector=&RandomRangeVector(Minimum=<0 0 50>, Maximum=<0 1 250>, Evaluation Type=::NewEnumerator1)
    )
    Coordinate Space=::Local
   )
   &InheritSourceMovement(
    Applied Position Scale=&Multiply_VectorByFloat(
     Vector=<1 1 1>
     Float=&FloatFromCurve(FloatCurve=DataInterface() {propertyValues {Curve {
         defaultValue=1
         keys [
          {arriveTangent=-0.8702, interpMode=RCIM_Cubic, leaveTangent=-0.8702, tangentMode=RCTM_User, time=0.05, value=1}
          {interpMode=RCIM_Cubic, tangentMode=RCTM_User, time=1, value=0}
         ]
        }}})
    )
    ApplyForce=false
    ApplyPosition=true
    ApplyPositionFalloff=false
   ) {module=InheritSourceMovement001}
   &SubUVAnimation(
    "Sprite Renderer "=DataInterface() {propertyValues {SpriteRenderer=&self:NE_Lantern_Sprite_01_0.NiagaraSpriteRendererProperties_0}}
    UseEndFrame=false
    UseStartFrame=false
   )
   &DynamicMaterialParameters(
    Index 0 Param 1=0.01
    Index 0 Param 2=&MakeCustomFloatFromBool(Bool=$User.bIsFrontendPreview, True Float=6, False Float=50)
    Index 0 Param 3=&MakeCustomFloatFromBool(Bool=$User.bIsFrontendPreview, True Float=1, False Float=0.8)
    Index 0 Param 4=&RandomRangeFloat(Minimum=0, Maximum=1)
    Module.Index0Float0Write=true
    Module.Index0Float1Write=true
    Module.Index0Float2Write=true
    Module.Index0Float3Write=true
   )
   &SolveForcesAndVelocity(Clamp Velocity=false, Limit Acceleration=false)
  ]
  rendererClasses [&NiagaraSpriteRendererProperties]
  renderers [{rendererClass=&NiagaraSpriteRendererProperties}]
 }
 NE_Lantern_Sprite_02 {
  emitterUpdateScript [
   &EmitterState()
   &VC_FilteredSocketCount()
   &SpawnRate(
    SpawnRate=&Multiply_Float(
     A=&MakeCustomFloatFromBool(Bool=$User.bIsFrontendPreview, True Float=50, False Float=100)
     B=&VC_Mask_Floats(Coro Value:NiagaraFloat=1, FrontEnd Value:NiagaraFloat=1, FN Value:NiagaraFloat=0.5)
    )
    Use Spawn Probability=false
   )
  ]
  particleSpawnScript [
   &VC_PaintedVariantsOverride() {enabled=false}
   &InitializeParticle(
    Color=<rgba 0.0885 0.0885 0.0885 1>
    Lifetime=&MakeCustomFloatFromBool(
     Bool=$User.bIsFrontendPreview
     True Float=&RandomRangeFloat(Minimum=0.5, Maximum=0.8)
     False Float=&RandomRangeFloat(Minimum=0.5, Maximum=1)
    )
    Sprite Rotation Mode=::NewEnumerator2
    Sprite Size=&MakeCustomVector2DFromBool(
     Bool=$User.bIsFrontendPreview
     True Vector2D=&RandomRangeVector2D(
      Minimum=&Multiply_Vector2D_ByFloat(Float=1.4, Vector2D=<4.5 10>)
      Maximum=&Multiply_Vector2D_ByFloat(Float=1.6, Vector2D=<1 8>)
     )
     False Vector2D=&RandomRangeVector2D(
      Minimum=&Multiply_Vector2D_ByFloat(Float=1.5, Vector2D=<4.5 10>)
      Maximum=&Multiply_Vector2D_ByFloat(Float=2, Vector2D=<1 8>)
     )
    )
    Sprite Size Mode=::NewEnumerator4
    UsePositionOffset=false
    Write Material Random=true
   )
   &PartitionParticles(Number Of Partitions=$Emitter.PortCount)
   &SkeletalMeshLocation(
    Bone / Socket Index=$Particles.PartitionParticles.Partition
    Bone / Socket Sampling Mode=::NewEnumerator5
    MaskSpawnGroup=false
    Mesh Sampling Type=::NewEnumerator6
    Skeletal Mesh=$System.SkeletalMesh
   )
   &AddVelocity(Velocity=&MakeCustomVectorFromBool(
     Bool=$User.bIsFrontendPreview
     True Vector=&RandomRangeVector(Minimum=<-1000 200 50>, Maximum=<-2000 -200 80>)
     False Vector=&RandomRangeVector(Minimum=<-600 200 0>, Maximum=<-1000 -200 50>)
    ))
   &ShapeLocation(MaskSpawnGroup=false, Override Seed=false, Sphere Radius=50, UseOwnerScale=true)
   &DLR_ProductPreviews(GameModeID:NiagaraInt32=$System.GameModeID, bPreviewVehicleSpeed:NiagaraBool=$System.bPreviewVehicleSpeed)
  ]
  particleUpdateScript [
   &ParticleState()
   &ScaleColor(
    Scale RGB=&MakeCustomVectorFromBool(
     Bool=$User.bIsFrontendPreview
     True Vector=&VectorFromCurve(
      VectorCurve=DataInterface() {propertyValues {
        XCurve {keys [<at 0 1 RCIM_Cubic>,<at 0.25 5 RCIM_Cubic>,<at 0.5 1 RCIM_Cubic>,<at 0.75 5 RCIM_Cubic>,<at 1 1 RCIM_Cubic>]}
        YCurve {keys [
          {interpMode=RCIM_Cubic, leaveTangentWeight=0.3333, tangentMode=RCTM_User, time=0, value=1}
          {arriveTangentWeight=0.0833, interpMode=RCIM_Cubic, leaveTangentWeight=0.25, tangentMode=RCTM_User, time=0.25, value=5}
          {arriveTangentWeight=0.0833, interpMode=RCIM_Cubic, leaveTangentWeight=0.1667, tangentMode=RCTM_User, time=0.5, value=1}
          {arriveTangentWeight=0.0833, interpMode=RCIM_Cubic, leaveTangentWeight=0.0833, tangentMode=RCTM_User, time=0.75, value=5}
          {arriveTangentWeight=0.3333, interpMode=RCIM_Cubic, tangentMode=RCTM_User, time=1, value=1}
         ]}
        ZCurve {keys [
          {interpMode=RCIM_Cubic, leaveTangentWeight=0.3333, tangentMode=RCTM_User, time=0, value=1}
          {arriveTangentWeight=0.0833, interpMode=RCIM_Cubic, leaveTangentWeight=0.25, tangentMode=RCTM_User, time=0.25, value=5}
          {arriveTangentWeight=0.0833, interpMode=RCIM_Cubic, leaveTangentWeight=0.1667, tangentMode=RCTM_User, time=0.5, value=1}
          {arriveTangentWeight=0.0833, interpMode=RCIM_Cubic, leaveTangentWeight=0.0833, tangentMode=RCTM_User, time=0.75, value=5}
          {arriveTangentWeight=0.3333, interpMode=RCIM_Cubic, tangentMode=RCTM_User, time=1, value=1}
         ]}
       }}
      Scale Curve=&Multiply_VectorByFloat(Vector=<1 1 1>, Float=1)
     )
     False Vector=&VectorFromCurve(
      VectorCurve=DataInterface() {propertyValues {
        XCurve {keys [<at 0 1.5 RCIM_Cubic>,<at 0.25 5 RCIM_Cubic>,<at 0.5 1.5 RCIM_Cubic>,<at 0.75 5 RCIM_Cubic>,<at 1 1.5 RCIM_Cubic>]}
        YCurve {keys [
          {interpMode=RCIM_Cubic, leaveTangentWeight=0.3333, tangentMode=RCTM_User, time=0, value=1.5}
          {arriveTangentWeight=0.0833, interpMode=RCIM_Cubic, leaveTangentWeight=0.25, tangentMode=RCTM_User, time=0.25, value=5}
          {arriveTangentWeight=0.0833, interpMode=RCIM_Cubic, leaveTangentWeight=0.1667, tangentMode=RCTM_User, time=0.5, value=1.5}
          {arriveTangentWeight=0.0833, interpMode=RCIM_Cubic, leaveTangentWeight=0.0833, tangentMode=RCTM_User, time=0.75, value=5}
          {arriveTangentWeight=0.3333, interpMode=RCIM_Cubic, tangentMode=RCTM_User, time=1, value=1.5}
         ]}
        ZCurve {keys [
          {interpMode=RCIM_Cubic, leaveTangentWeight=0.3333, tangentMode=RCTM_User, time=0, value=1.5}
          {arriveTangentWeight=0.0833, interpMode=RCIM_Cubic, leaveTangentWeight=0.25, tangentMode=RCTM_User, time=0.25, value=5}
          {arriveTangentWeight=0.0833, interpMode=RCIM_Cubic, leaveTangentWeight=0.1667, tangentMode=RCTM_User, time=0.5, value=1.5}
          {arriveTangentWeight=0.0833, interpMode=RCIM_Cubic, leaveTangentWeight=0.0833, tangentMode=RCTM_User, time=0.75, value=5}
          {arriveTangentWeight=0.3333, interpMode=RCIM_Cubic, tangentMode=RCTM_User, time=1, value=1.5}
         ]}
       }}
      Scale Curve=&Multiply_VectorByFloat(Vector=<1 1 1>, Float=1)
     )
    )
    ScaleA=true
    ScaleRGB=true
   )
   &InheritSourceMovement(
    Applied Position Scale=&Multiply_VectorByFloat(
     Vector=<1 1 1>
     Float=&FloatFromCurve(FloatCurve=DataInterface() {propertyValues {Curve {
         defaultValue=1
         keys [
          {arriveTangent=-1.7885, interpMode=RCIM_Cubic, leaveTangent=-1.7885, tangentMode=RCTM_User, time=0.05, value=1}
          {interpMode=RCIM_Cubic, tangentMode=RCTM_User, time=1, value=0}
         ]
        }}})
    )
    ApplyForce=false
    ApplyPosition=true
    ApplyPositionFalloff=false
   ) {module=InheritSourceMovement001}
   &AccelerationForce(Acceleration=&VectorFromCurve(
     VectorCurve=DataInterface() {propertyValues {XCurve {keys [<at 0 0>,<at 1 0>]}, YCurve {keys [<at 0 0>,<at 1 0>]}, ZCurve {keys [<at 0 0 RCIM_Cubic>,<at 1 -1 RCIM_Cubic>]}}}
     Scale Curve=&MakeCustomVectorFromBool(Bool=$User.bIsFrontendPreview, True Vector=<0 0 3000>, False Vector=<1 0 1000>)
    ))
   &SolveForcesAndVelocity(Clamp Velocity=false, Limit Acceleration=false)
  ]
  rendererClasses [&NiagaraSpriteRendererProperties]
  renderers [{rendererClass=&NiagaraSpriteRendererProperties}]
 }
}
summary {
 systemName=NS_Boost_Lantern
 emitters [
  {emitterName=NE_Lantern_Sprite_01, rendererClasses [&NiagaraSpriteRendererProperties]}
  {emitterName=NE_Lantern_Sprite_02, rendererClasses [&NiagaraSpriteRendererProperties]}
 ]
 userVariables [
  User.bIsFrontendPreview {type=NiagaraBool}
  User.bIsPainted {type=NiagaraBool}
  User.blsIsCoroColor {type=NiagaraBool}
  User.bPreviewVehicleSpeed {type=NiagaraBool}
  User.ColorVariantID {type=NiagaraInt32}
  User.GameModeID {type=NiagaraInt32}
  User.PaintedColor {type=LinearColor, description="", defaultValue {struct=LinearColor, value=<rgba 1 1 1 1>}}
  User.PaintedScalar {type=NiagaraFloat, defaultValue=1, description=""}
  User.PortCount {type=NiagaraInt32, defaultValue=2}
 ]
}
